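# Editor utility widget that helps environment artists find, select and hide Fortnite building actors in a level.
# Buttons gather actors by class lists, CanBeDamaged or texture data, then select them or toggle temporary editor visibility through shared helpers.
# The wireframe toggle switches off Nanite and applies a dynamic overlay material, optionally with flat gray base materials.
# A username check reveals extra controls, including updating the world's time-of-day manager.
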EditorUtilityWidgetBlueprint /Game/ArtTools/Environment/EnvironmentActorFinder/EUW_EnvActorFinder
graphs [
 ClearSelectedLevelActors {nodes [
   ClearSelectedLevelActors() {is=title K2Node_FunctionEntry}
   > SetSelectedLevelActors(self=Get EditorActorSubsystem(), ActorsToSelect=Get AR_EmptyActors() -> (AR_EmptyActors:Array of Actor Object References))
  ]}
 ClearValidActorsArray {nodes [
   ClearValidActorsArray() {is=title K2Node_FunctionEntry}
   > Clear(TargetArray:Array of Actor Object References=Get AR_ValidActors() -> (AR_ValidActors:Array of Actor Object References))
  ]}
 CreateCollection {nodes [
   n1 = CreateCollection!() -> (then, CollectionName:Name) {is=title K2Node_FunctionEntry}
   > n2 = Set CollectionName(CollectionName:Name=@n1) -> (Output_Get:Name=None)
   > Return Node!(execute, Output_Get:Name=@n2)
  ]}
 EventGraph {nodes [
   Clear(TargetArray=Get AR_ValidActors())
   > n1 = For Each Loop(Array:Array of Actor Class References=Get AR_ValidArchitectureClasses() -> (AR_ValidArchitectureClasses:Array of Actor Class References)) -> (Array Element:Actor Class Reference)
   > n2 = GetAllActorsOfClass!(execute, WorldContextObject:Object Reference, ActorClass:Actor Class Reference=@n1) -> (then, OutActors:Array of Actor Object References)
   > n3 = For Each Loop(Array:Array of Actor Object References=@n2) -> (Array Element:Actor Object Reference)
   n4 = Get AR_ValidActors()
   SetSelectedLevelActors(execute=@n3.Completed, self=Get EditorActorSubsystem(), ActorsToSelect=@n4)
   Add(execute=@n3.LoopBody, TargetArray:Array of Actor Object References=@n4, NewItem:"Actor Object Reference (by ref)"=@n3) {is=title K2Node_CallArrayFunction}
   Clear(TargetArray=Get AR_ValidActors())
   > n5 = For Each Loop(Array:Array of Actor Class References=Get AR_ValidArchitectureClasses()) -> (Array Element:Actor Class Reference)
   > n6 = GetAllActorsOfClass(ActorClass=@n5)
   > n7 = For Each Loop(Array:Array of Actor Object References=@n6) -> (Array Element:Actor Object Reference)
   n8 = NOT Boolean(A=false)
   > SetIsTemporarilyHiddenInEditor!(execute, self:Actor Object Reference=@n7, bIsHidden:bool=@n8) -> (then)
   > n9 = Branch(Condition=@n8)
   n10 = Get BTN_HideAllArchitecture() -> (BTN_HideAllArchitecture:TA Generic 01 Button Object Reference)
   > setLabel!(execute, self:TA Base Editor Button Widget Object Reference=@n10, label:String=Show Architecture) -> (then)
   setLabel(execute=@n9.else, self=@n10, label=Hide Architecture)
   Clear(TargetArray=Get AR_ValidActors())
   > n11 = For Each Loop(Array:Array of Actor Class References=Get AR_ValidPropClasses() -> (AR_ValidPropClasses:Array of Actor Class References)) -> (Array Element:Actor Class Reference)
   > n12 = GetAllActorsOfClass(ActorClass=@n11)
   > n13 = For Each Loop(Array:Array of Actor Object References=@n12) -> (Array Element:Actor Object Reference)
   n14 = Get AR_ValidActors()
   SetSelectedLevelActors(execute=@n13.Completed, self=Get EditorActorSubsystem(), ActorsToSelect=@n14)
   Add(execute=@n13.LoopBody, TargetArray=@n14, NewItem=@n13)
   n15 = Get EditorActorSubsystem()
   n16 = Get AR_EmptyActors()
   SetSelectedLevelActors(self=@n15, ActorsToSelect=@n16)
   > n17 = For Each Loop(Array:Array of Actor Class References=Get AR_ValidPropClasses()) -> (Array Element:Actor Class Reference)
   > n18 = GetAllActorsOfClass(ActorClass=@n17)
   > n19 = For Each Loop(Array:Array of Actor Object References=@n18) -> (Array Element:Actor Object Reference)
   n20 = Get RootComponent!(self:Actor Object Reference=@n19) -> (RootComponent:Scene Component Object Reference)
   > n21 = GetEditorProperty!(execute:Exec, Object:Object Reference=@n20, PropertyName:Name=bvisible) -> (then:Exec, PropertyValue:bool, ReturnValue:bool)
   > n22 = Branch(Condition=@n21)
   ToggleVisibility!(execute=@n22.else, self:Scene Component Object Reference=@n20, bPropagateToChildren:bool=true) -> (then)
   SetSelectedLevelActors(self=@n15, ActorsToSelect=@n16)
   > n23 = For Each Loop(Array:Array of Actor Class References=Get AR_ValidPropClasses()) -> (Array Element:Actor Class Reference)
   > n24 = GetAllActorsOfClass(ActorClass=@n23)
   > n25 = For Each Loop(Array:Array of Actor Object References=@n24) -> (Array Element:Actor Object Reference)
   n26 = Get RootComponent(self=@n25)
   > n27 = GetEditorProperty(Object=@n26, PropertyName=bvisible)
   > Branch(Condition=@n27)
   > ToggleVisibility(self=@n26, bPropagateToChildren=true)
   n28 = Get EditorActorSubsystem()
   n29 = Get AR_EmptyActors()
   SetSelectedLevelActors(self=@n28, ActorsToSelect=@n29)
   > Clear(TargetArray=Get AR_ValidActors())
   > n30 = For Each Loop(Array:Array of Actor Class References=Get AR_ValidArchitectureClasses()) -> (Array Element:Actor Class Reference)
   > n31 = GetAllActorsOfClass(ActorClass=@n30)
   > n32 = For Each Loop(Array:Array of Actor Object References=@n31) -> (Array Element:Actor Object Reference)
   n33 = Get RootComponent(self=@n32)
   > n34 = GetEditorProperty(Object=@n33, PropertyName=bvisible)
   > n35 = Branch(Condition=@n34)
   ToggleVisibility(execute=@n35.else, self=@n33, bPropagateToChildren=true)
   Set Actor Hidden In Game!(execute, self:Actor Object Reference, bNewHidden:bool=true) -> (then) {is=title K2Node_CallFunction}
   > Add Unique(TargetArray:Array of Actor Object References=Get AR_ValidActors(), NewItem:"Actor Object Reference (by ref)"=For Each Loop())
   n36 = Get AssetTagsSubsystem() -> (ReturnValue:Asset Tags Subsystem Object Reference) {is=title K2Node_GetEngineSubsystem}
   n37 = Get CollectionName() -> (CollectionName:Name=None)
   n38 = CollectionExists!(execute, self:Asset Tags Subsystem Object Reference=@n36, Name:Name=@n37) -> (then, ReturnValue:bool)
   > n39 = Branch(Condition=@n38)
   > n40 = EmptyCollection!(execute, self:Asset Tags Subsystem Object Reference=@n36, Name:Name=@n37) -> (then, ReturnValue:bool)
   Add Assets To Collection!(
    execute=[@n41.then, @n40.then]
    self:Asset Tags Subsystem Object Reference=@n36
    Name:Name=@n37
    AssetPaths:Array of Soft Object Path Structures=Get AR_SoftPaths() -> (AR_SoftPaths:Array of Soft Object Path Structures)
   ) -> (then, ReturnValue:bool)
   n41 = CreateCollection!(execute=@n39.else, self:Asset Tags Subsystem Object Reference=@n36, Name:Name=@n37, ShareType:ECollectionScriptingShareType Enum=Local) -> (then, ReturnValue:bool) {is=title K2Node_CallFunction}
   ExecuteConsoleCommand(Command=r.nanite 0)
   > n42 = GetAllActorsOfClass(ActorClass=/Script/FortniteGame.BuildingActor) -> (OutActors:Array of Building Actor Object References)
   > n43 = For Each Loop(Array=@n42)
   > n44 = For Each Loop(Array:Array of Static Mesh Component Object References=Get Components By Class!(self:Actor Object Reference=@n43, ComponentClass:Actor Component Class Reference=&StaticMeshComponent) -> (ReturnValue:Array of Static Mesh Component Object References)) -> (Array Element:Static Mesh Component Object Reference)
   > SetOverlayMaterial!(
    execute
    self:Mesh Component Object Reference=@n44
    NewOverlayMaterial:Material Interface Object Reference=Get OverlayMaterial() -> (OverlayMaterial:Material Interface Object Reference)
    bSetMaterialSlot:bool=false
    SlotIndex:int=0
   ) -> (then)
   "On Clicked (BTN_SelectArchitecture)"!() -> (OutputDelegate, then, widget:User Widget Object Reference, event:Pointer Event Structure) {is=title K2Node_ComponentBoundEvent}
   > ResetSelectedActors()
   > SelectValidActors!(execute, self, ActorClasses:Array of Actor Class References=Get AR_ValidArchitectureClasses()) -> (then)
   "On Clicked (BTN_SelectProps)"!() -> (OutputDelegate, then, widget:User Widget Object Reference, event:Pointer Event Structure) {is=title K2Node_ComponentBoundEvent}
   > ResetSelectedActors()
   > SelectValidActors(ActorClasses=Get AR_ValidPropClasses())
   n45 = "On Selection Changed (BTN_EnableOverlay)"!() -> (OutputDelegate, then, selected:bool, widget:User Widget Object Reference) {is=title K2Node_ComponentBoundEvent}
   > n46 = Set IsWireframeEnabled(IsWireframeEnabled:bool=@n45.selected) -> (Output_Get:bool)
   > ResetSelectedActors()
   > n47 = Branch(Condition=@n46)
   > ExecuteConsoleCommand(Command=r.nanite 0)
   > n48 = CreateDynamicMaterialInstance(Parent=Get OverlayMaterial(), OptionalName=DynamicWireframeMaterial, CreationFlags=None)
   > n49 = Set NewOverlayMaterial(NewOverlayMaterial:Material Instance Dynamic Object Reference=@n48) -> (Output_Get:Material Instance Dynamic Object Reference)
   > SetVectorParameterValue(
    self=@n49
    ParameterName=WireframeColor
    Value="To LinearColor (Color)"!(InColor:Color Structure=Get WireframeColor() -> (WireframeColor:Color Structure)) -> (ReturnValue:Linear Color Structure)
   )
   > SetScalarParameterValue(self=@n49, ParameterName=EmissiveIntensity, Value=Get EmissiveIntensity() -> (EmissiveIntensity:float))
   > n50 = For Each Loop(Array:Array of Actor Class References=Make Array!(
     "[0]":Actor Class Reference=&BuildingActor
     "[1]":Actor Class Reference=/Script/FortniteGame.BuildingStaticGeo
     "[2]":Actor Class Reference=/Script/Engine.SkeletalMeshActor
    ) -> (Array:Array of Actor Class References)) -> (Array Element:Actor Class Reference)
   > n51 = GetAllActorsOfClass(ActorClass=@n50) -> (OutActors:Array of Actor Object References)
   > n52 = Set AR_EnvActors(AR_EnvActors:Array of Actor Object References=@n51) -> (Output_Get:Array of Actor Object References)
   > n53 = For Each Loop(Array:Array of Actor Object References=@n52) -> (Array Element:Actor Object Reference)
   > n54 = Set ModifiedActor(ModifiedActor:Actor Object Reference=@n53) -> (Output_Get:Actor Object Reference)
   > n55 = For Each Loop(Array:Array of Static Mesh Component Object References=Get Components By Class(self=@n54, ComponentClass=&StaticMeshComponent)) -> (Array Element:Static Mesh Component Object Reference)
   > n56 = Set ModifiedMeshComponent(ModifiedMeshComponent:Static Mesh Component Object Reference=@n55) -> (Output_Get:Static Mesh Component Object Reference)
   > SetOverlayMaterial(
    self=@n56
    NewOverlayMaterial=Get NewOverlayMaterial() -> (NewOverlayMaterial:Material Instance Dynamic Object Reference)
    bSetMaterialSlot=false
    SlotIndex=0
   )
   > Branch(Condition=Get ReplaceBaseMaterial() -> (ReplaceBaseMaterial:bool))
   n57 = Get ModifiedMeshComponent() -> (ModifiedMeshComponent:Static Mesh Component Object Reference)
   > For Loop(LastIndex=int - int(A=GetNumMaterials!(self:Primitive Component Object Reference=@n57) -> (ReturnValue:int=0)))
   > SetMaterial(self=Get ModifiedMeshComponent(), ElementIndex=0, Material=/Game/ArtTools/Environment/OverlayMaterials/MI_FlatGray_A)
   ExecuteConsoleCommand(execute=@n47.else, Command=r.nanite 1)
   > n58 = For Each Loop(Array:Array of Actor Class References=Make Array("[0]"=&BuildingActor, "[1]"=&BuildingStaticGeo, "[2]"=&SkeletalMeshActor)) -> (Array Element:Actor Class Reference)
   > n59 = GetAllActorsOfClass(ActorClass=@n58)
   > n60 = For Each Loop(Array:Array of Actor Object References=@n59) -> (Array Element:Actor Object Reference)
   > Sequence()
   > n61 = For Each Loop(Array:Array of Static Mesh Component Object References=Get Components By Class(self=@n60, ComponentClass=&StaticMeshComponent)) -> (Array Element:Static Mesh Component Object Reference)
   > SetOverlayMaterial(self=@n61, bSetMaterialSlot=false, SlotIndex=0)
   > Branch(Condition=Get ReplaceBaseMaterial())
   > n62 = For Loop(LastIndex=int - int(A=GetNumMaterials(self=@n61) -> (ReturnValue=0)))
   > SetMaterial(self=@n61, ElementIndex=@n62)
   n63 = "On Selection Changed (BTN_HideStaticGeo)"!() -> (OutputDelegate, then, selected:bool, widget:User Widget Object Reference) {is=title K2Node_ComponentBoundEvent}
   > n64 = Branch(Condition=@n63.selected)
   > ResetSelectedActors()
   n65 = Make Array!("[0]"=Get BuildingStaticGeo() -> (BuildingStaticGeo:Actor Class Reference)) -> (Array)
   > ToggleEditorVisibility!(execute, self, ActorClasses:Array of Actor Class References=@n65, bIsHidden:bool=true) -> (then)
   n66 = Get BTN_HideStaticGeo() -> (BTN_HideStaticGeo:TA Generic 01 Button Object Reference)
   > setLabel(self=@n66, label=Show Static Geo)
   ToggleEditorVisibility(execute=@n64.else, ActorClasses=@n65, bIsHidden=false)
   > setLabel(self=@n66, label=Hide Static Geo)
   "On Property Changed (SPV_WireframeColor)"!() -> (OutputDelegate, then, PropertyName:Name=None) {is=title K2Node_ComponentBoundEvent}
   > SetVectorParameterValue(self=Get NewOverlayMaterial(), ParameterName=WireframeColor, Value="To LinearColor (Color)"(InColor=Get WireframeColor()))
   "On Property Changed (SPV_ReplaceBaseMat)"!() -> (OutputDelegate, then, PropertyName:Name=None) {is=title K2Node_ComponentBoundEvent}
   > ResetSelectedActors()
   > Branch(Condition=Get IsWireframeEnabled() -> (IsWireframeEnabled:bool))
   > n67 = Branch(Condition=Get ReplaceBaseMaterial())
   > n68 = For Each Loop(Array:Array of Actor Class References=Make Array!(
     "[0]"=&BuildingActor
     "[1]":Actor Class Reference=&BuildingStaticGeo
     "[2]":Actor Class Reference=&SkeletalMeshActor
    ) -> (Array)) -> (Array Element:Actor Class Reference)
   > n69 = GetAllActorsOfClass(ActorClass=@n68)
   > n70 = Set AR_EnvActors(AR_EnvActors=@n69)
   > n71 = For Each Loop(Array:Array of Actor Object References=@n70) -> (Array Element:Actor Object Reference)
   > n72 = For Each Loop(Array:Array of Static Mesh Component Object References=Get Components By Class(self=@n71, ComponentClass=&StaticMeshComponent)) -> (Array Element:Static Mesh Component Object Reference)
   > SetOverlayMaterial(self=@n72, NewOverlayMaterial=Get NewOverlayMaterial(), bSetMaterialSlot=false, SlotIndex=0)
   > Branch(Condition=Get ReplaceBaseMaterial())
   > n73 = For Loop(LastIndex=int - int(A=GetNumMaterials(self=@n72) -> (ReturnValue=0)))
   > SetMaterial(self=@n72, ElementIndex=@n73, Material=&MI_FlatGray_A)
   n74 = For Each Loop(
    Exec=@n67.else
    Array:Array of Actor Class References=Make Array("[0]"=&BuildingActor, "[1]"=&BuildingStaticGeo, "[2]"=&SkeletalMeshActor)
   ) -> (Array Element:Actor Class Reference)
   > n75 = GetAllActorsOfClass(ActorClass=@n74)
   > n76 = Set AR_EnvActors(AR_EnvActors=@n75)
   > n77 = For Each Loop(Array:Array of Actor Object References=@n76) -> (Array Element:Actor Object Reference)
   > n78 = For Each Loop(Array:Array of Static Mesh Component Object References=Get Components By Class(self=@n77, ComponentClass=&StaticMeshComponent)) -> (Array Element:Static Mesh Component Object Reference)
   > n79 = For Loop(LastIndex=int - int(A=GetNumMaterials(self=@n78) -> (ReturnValue=0)))
   > SetMaterial(self=@n78, ElementIndex=@n79)
   "On Clicked (BTN_SelectStaticMeshes)"!() -> (OutputDelegate, then, widget:User Widget Object Reference, event:Pointer Event Structure) {is=title K2Node_ComponentBoundEvent}
   > ResetSelectedActors()
   > Clear(TargetArray=Get AR_BSG_Actors() -> (AR_BSG_Actors:Array of Actor Object References))
   > Clear(TargetArray=Get AR_FSM_Actors() -> (AR_FSM_Actors:Array of Actor Object References))
   > GetValidStaticMeshes()
   "On Clicked (BTN_UpdateTODM)"!() -> (OutputDelegate, then, widget:User Widget Object Reference, event:Pointer Event Structure) {is=title K2Node_ComponentBoundEvent}
   > SetEditorProperty(
    Object=GetFortWorldSettingsInEditor!(WorldContextObject:Object Reference) -> (ReturnValue:Fort World Settings Object Reference)
    PropertyName=WorldTimeOfDaymanager
    PropertyValue:Info Soft Class Reference=Get CurrentTODM() -> (CurrentTODM:Info Soft Class Reference)
   )
   n80 = Get BTN_UpdateTODM() -> (BTN_UpdateTODM:TA Generic 01 Button Object Reference)
   > Set buttonColor_Default!(
    execute
    buttonColor_Default:Linear Color Structure=(R=0.1354,G=0.1354,B=0.1354,A=1)
    self:TA Base Editor Button Widget Object Reference=@n80
   ) -> (then, Output_Get:Linear Color Structure)
   > setLabel(self=@n80, label=TODM Already Updated)
   "On Clicked (BTN_EditWidget)"() {is=title K2Node_ComponentBoundEvent}
   > n81 = GetBlueprintAsset(Object=&self)
   > Execute Python Script!(
    execute
    PythonScript="import unreal\r\n\r\nmy_widget = widget_object\r\nasset_editor_subsystem = unreal.get_editor_subsystem(unreal.AssetEditorSubsystem)\r\nasset_editor_subsystem.open_editor_for_assets([my_widget])"
    widget_object:Blueprint Object Reference=@n81
   ) -> (then, ReturnValue) {inputs [widget_object]}
   "On Clicked (BTN_SelectByProperty)"!() -> (OutputDelegate, then, widget:User Widget Object Reference, event:Pointer Event Structure) {is=title K2Node_ComponentBoundEvent}
   > ResetSelectedActors()
   > n82 = For Each Loop(Array:Array of Actor Class References=Get AR_ValidDamageClasses() -> (AR_ValidDamageClasses:Array of Actor Class References)) -> (Array Element:Actor Class Reference)
   > n83 = GetAllActorsOfClass(ActorClass=@n82)
   > n84 = For Each Loop(Array:Array of Actor Object References=@n83) -> (Array Element:Actor Object Reference)
   n85 = Get AR_ValidActors()
   SetSelectedLevelActors(execute=@n84.Completed, self=Get EditorActorSubsystem(), ActorsToSelect=@n85)
   n86 = GetEditorProperty(execute=@n84.LoopBody, Object=@n84, PropertyName=bCanBeDamaged)
   > n87 = Branch(Condition=@n86)
   Add(execute=@n87.else, TargetArray=@n85, NewItem=@n84)
   "On Clicked (BTN_SelectByTextureData)"!() -> (OutputDelegate, then, widget:User Widget Object Reference, event:Pointer Event Structure) {is=title K2Node_ComponentBoundEvent}
   > Set SelectedTextureData(SelectedTextureData:Building Texture Data Object Reference=Get TextureData() -> (TextureData:Building Texture Data Object Reference)) -> (Output_Get:Building Texture Data Object Reference)
   > n88 = Sequence()
   > ResetSelectedActors()
   > n89 = GetAllActorsOfClass(ActorClass=&BuildingActor) -> (OutActors:Array of Building Actor Object References)
   > n90 = For Each Loop(Array:Array of Building Actor Object References=@n89) -> (Array Element:Building Actor Object Reference)
   > Add(
    TargetArray:Array of Building Actor Object References=Get AR_Actors() -> (AR_Actors:Array of Building Actor Object References)
    NewItem:"Building Actor Object Reference (by ref)"=@n90
   )
   n91 = For Each Loop(Exec=@n88.then_1, Array=Get AR_Actors())
   SetSelectedLevelActors(execute=@n91.Completed, self=Get EditorActorSubsystem(), ActorsToSelect=Get AR_ValidActors())
   n92 = Set CurrentBuildingActor(execute=@n91.LoopBody, CurrentBuildingActor:Building Actor Object Reference=@n91) -> (Output_Get:Building Actor Object Reference)
   > n93 = Cast To BuildingSMActor!(execute, Object=@n92) -> (then, CastFailed, AsBuilding Static Mesh:Building Static Mesh Object Reference) {is=title K2Node_DynamicCast}
   > n94 = GetBuildingTextureData!(execute, self:Building Static Mesh Object Reference=@n93) -> (then, OutTextureDataArray:Array of Building Texture Data Object References)
   > Branch(Condition="integer > integer"(A=Length(TargetArray:Array of Building Texture Data Object References=@n94)))
   > n95 = For Each Loop(Array:Array of Building Texture Data Object References=@n94) -> (Array Element:Building Texture Data Object Reference)
   > Branch(Condition="Equal (Object)"(A=@n95, B=Get SelectedTextureData() -> (SelectedTextureData:Building Texture Data Object Reference)))
   > Add(
    TargetArray:Array of Actor Object References=Get AR_ValidActors()
    NewItem:"Actor Object Reference (by ref)"=Get CurrentBuildingActor() -> (CurrentBuildingActor:Building Actor Object Reference)
   )
   n96 = "On Selection Changed (BTN_HideAllArchitecture)"!() -> (OutputDelegate, then, selected:bool, widget:User Widget Object Reference) {is=title K2Node_ComponentBoundEvent}
   > n97 = Branch(Condition=@n96.selected)
   > ResetSelectedActors()
   n98 = Get AR_ValidArchitectureClasses()
   > ToggleEditorVisibility(ActorClasses=@n98, bIsHidden=true)
   n99 = Get BTN_HideAllArchitecture()
   > setLabel(self=@n99, label=Show Architecture)
   ToggleEditorVisibility(execute=@n97.else, ActorClasses=@n98, bIsHidden=false)
   > setLabel(self=@n99, label=Hide Architecture)
   n100 = "On Selection Changed (BTN_HideAllProps)"!() -> (OutputDelegate, then, selected:bool, widget:User Widget Object Reference) {is=title K2Node_ComponentBoundEvent}
   > n101 = Branch(Condition=@n100.selected)
   > ResetSelectedActors()
   n102 = Get AR_ValidPropClasses()
   > ToggleEditorVisibility(ActorClasses=@n102, bIsHidden=true)
   n103 = Get BTN_HideAllProps() -> (BTN_HideAllProps:TA Generic 01 Button Object Reference)
   > setLabel(self=@n103, label=Show Props)
   ToggleEditorVisibility(execute=@n101.else, ActorClasses=@n102, bIsHidden=false)
   > setLabel(self=@n103, label=Hide Props)
   "On Clicked (BTN_SelectStaticGeo)"!() -> (OutputDelegate, then, widget:User Widget Object Reference, event:Pointer Event Structure) {is=title K2Node_ComponentBoundEvent}
   > ResetSelectedActors()
   > SelectValidActors(ActorClasses=Make Array!("[0]"=Get BuildingStaticGeo()) -> (Array))
   Event PreConstruct()
   > SetObject(self=Get SPV_TextureData() -> (SPV_TextureData:Single Property View Object Reference), NewObject=Self-Reference())
   Event Construct()
   > n104 = Sequence()
   > n105 = Sequence!(execute) -> (then_0, then_1, then_2:Exec)
   > n106 = Branch(Condition="Equal Exactly (String)"(A=GetPlatformUserName!() -> (ReturnValue:String), B=brianfasten))
   n107 = Get BTN_EditWidget() -> (BTN_EditWidget:TA Generic 01 Button Object Reference)
   > SetVisibility(self=@n107, InVisibility=Visible)
   n108 = Get BTN_UpdateTODM()
   > SetVisibility(self=@n108, InVisibility=Visible)
   n109 = Get VB_Version2() -> (VB_Version2:Vertical Box Object Reference)
   > SetVisibility(self=@n109, InVisibility=Hidden)
   n110 = Get SPV_ReplaceBaseMat() -> (SPV_ReplaceBaseMat:Single Property View Object Reference)
   > SetVisibility(self=@n110, InVisibility=Visible)
   SetVisibility(execute=@n106.else, self=@n107, InVisibility=Hidden)
   > SetVisibility(self=@n108, InVisibility=Hidden)
   > SetVisibility(self=@n109, InVisibility=Hidden)
   > SetVisibility(self=@n110, InVisibility=Hidden)
   n111 = Self-Reference()
   SetObject(execute=@n105.then_2, self=Get SPV_WireframeColor() -> (SPV_WireframeColor:Single Property View Object Reference), NewObject=@n111)
   > SetObject(self=Get SPV_ReplaceBaseMat(), NewObject=@n111)
   > SetObject(self=Get SPV_TextureData(), NewObject=@n111)
   n112 = GetEditorProperty(execute=@n105.then_1, Object=GetFortWorldSettingsInEditor(), PropertyName=WorldTimeOfDaymanager) -> (PropertyValue:"Object Soft Class Reference (by ref)")
   > Branch(Condition="Equal (SoftClassReference)"(A:"Object Soft Class Reference (by ref)"=@n112, B:"Object Soft Class Reference (by ref)"=Get CurrentTODM()) -> (ReturnValue:bool=~) {is=title K2Node_PromotableOperator})
   n113 = Get BTN_UpdateTODM()
   > Set buttonColor_Default(buttonColor_Default=(R=0.1354,G=0.1354,B=0.1354,A=1), self=@n113)
   > setLabel(self=@n113, label=TODM Already Updated)
   SetObject(execute=@n104.then_1, self=Get DV_ActorClasses() -> (DV_ActorClasses:Details View Object Reference), NewObject=Self-Reference())
   n114 = Get RootComponent(self=@n7)
   n115 = GetEditorProperty(Object=@n114, PropertyName=bvisible) -> (PropertyValue:bool)
   > Branch(Condition=@n115)
   > ToggleVisibility(self=@n114, bPropagateToChildren=true)
   n116 = For Each Loop(Array:Array of Actor Object References=Get AR_ValidActors()) -> (Array Element:Actor Object Reference)
   n117 = Get Class Display Name(Class=Get Class(Object=@n116))
   > Add Unique(
    TargetArray:Array of Soft Object Path Structures=Get AR_SoftPaths()
    NewItem:"Soft Object Path Structure (by ref)"=MakeSoftObjectPath!(PathString:String=Split(SourceString=@n117, InStr=_C, SearchDir=FromStart)) -> (ReturnValue:Soft Object Path Structure)
   )
   n118 = Get BTN_SelectArchitecture() -> (BTN_SelectArchitecture:TA Generic 01 Button Object Reference)
   setLabel(self=@n118, label=Select Architecture)
   n119 = Get HorizontalBox_1() -> (HorizontalBox_1:Horizontal Box Object Reference)
   SetVisibility(self=@n119, InVisibility=Hidden)
   SetSelectedLevelActors(self=@n28, ActorsToSelect=@n29)
   GetMaterials!(self:Mesh Component Object Reference=@n57) -> (ReturnValue:Array of Material Interface Object References)
   n120 = Get BTN_SelectProps() -> (BTN_SelectProps:TA Generic 01 Button Object Reference)
   setLabel(self=@n120, label=Deselect Props)
   setLabel(self=@n120, label=Select Props)
   setLabel(self=@n118, label=Deselect Architecture)
   SetVisibility(self=@n119, InVisibility=Hidden)
   SetSelectedLevelActors(self=Get EditorActorSubsystem(), ActorsToSelect=Get AR_ValidActors())
   "On Property Changed (SPV_TextureData)"!() -> (OutputDelegate, then, PropertyName:Name=None) {is=title K2Node_ComponentBoundEvent}
   Get ModifiedActor() -> (ModifiedActor:Actor Object Reference)
  ]}
 GetActorClasses {nodes [
   GetActorClasses() {is=title K2Node_FunctionEntry}
   > n1 = GetAssets!(
    execute
    self:Asset Registry Interface=GetAssetRegistry()
    Filter:"ARFilter Structure (by ref)"=MakeARFilter!(
     PackageNames:Array of Names
     PackagePaths:Array of Names
     SoftObjectPaths:Array of Soft Object Path Structures
     ClassPaths:Array of Top Level Asset Path Structures
     RecursiveClassPathsExclusionSet:Set of Top Level Asset Path Structures
     ClassNames:Array of Names
     RecursiveClassesExclusionSet:Set of Names
     bRecursivePaths:bool=true
     bRecursiveClasses:bool=false
     bIncludeOnlyOnDiskAssets:bool=false
    ) -> (ReturnValue:ARFilter Structure)
    bSkipARFilteredAssets:bool=false
   ) -> (then, OutAssetData:Array of Asset Data Structures, ReturnValue:bool)
   > n2 = For Each Loop(Array:Array of Asset Data Structures=@n1) -> (Array Element:Asset Data Structure)
   > Add Unique(
    TargetArray:Array of Asset Data Structures=Get AR_ActorClasses() -> (AR_ActorClasses:Array of Asset Data Structures)
    NewItem:"Asset Data Structure (by ref)"=@n2
   )
   MakeTopLevelAssetPath!(PackageName:String=/Script/Engine, AssetName:String=World) -> (ReturnValue:Top Level Asset Path Structure)
   Make Array("[0]":Wildcard) -> (Array:Array of Wildcards)
  ]}
 GetValidStaticMeshes {nodes [
   GetValidStaticMeshes() {is=title K2Node_FunctionEntry}
   > n1 = For Each Loop(Array:Array of Actor Class References=Make Array("[0]":Actor Class Reference=Get StaticMeshClass() -> (StaticMeshClass:Actor Class Reference)) -> (Array:Array of Actor Class References)) -> (Array Element:Actor Class Reference)
   > n2 = GetAllActorsOfClass(ActorClass=@n1) -> (OutActors:Array of Actor Object References)
   > n3 = Set AR_FSM_Actors(AR_FSM_Actors:Array of Actor Object References=@n2) -> (Output_Get:Array of Actor Object References)
   > n4 = Filter by Class!(
    execute
    TargetArray:Array of Object References=@n3
    ObjectClass:Object Class Reference=Get BuildingStaticGeo()
    FilterType:EEditorScriptingFilterType Enum=Exclude
   ) -> (then, ReturnValue:Array of Actor Object References)
   > SetSelectedLevelActors(self=Get EditorActorSubsystem(), ActorsToSelect=@n4)
  ]}
 ReapplyDataOverrides {nodes [
   ReapplyDataOverrides() {is=title K2Node_FunctionEntry}
   > n1 = For Each Loop(Array=Make Array!("[0]"=&BuildingActor, "[1]":Actor Class Reference=&BuildingStaticGeo) -> (Array))
   > n2 = GetAllActorsOfClass(ActorClass=@n1)
   > n3 = For Each Loop(Array:Array of Actor Object References=@n2) -> (Array Element:Actor Object Reference)
   GetMaterial!(
    self:Primitive Component Object Reference=For Each Loop(Array:Array of Static Mesh Component Object References=Get Components By Class(self=@n3, ComponentClass=&StaticMeshComponent)) -> (Array Element:Static Mesh Component Object Reference)
    ElementIndex:int=0
   ) -> (ReturnValue:Material Interface Object Reference)
  ]}
 ResetSelectedActors {nodes [
   ResetSelectedActors() {is=title K2Node_FunctionEntry}
   > ClearSelectedLevelActors() {is=title K2Node_CallFunction}
   > ClearValidActorsArray() {is=title K2Node_CallFunction}
  ]}
 SelectValidActors {nodes [
   n1 = SelectValidActors!() -> (then, ActorClasses:Array of Actor Class References) {is=title K2Node_FunctionEntry}
   > n2 = For Each Loop(Array:Array of Actor Class References=@n1) -> (Array Element:Actor Class Reference)
   > n3 = GetAllActorsOfClass(ActorClass=@n2)
   > n4 = For Each Loop(Array:Array of Actor Object References=@n3) -> (Array Element:Actor Object Reference)
   n5 = Get AR_ValidActors()
   SetSelectedLevelActors(execute=@n4.Completed, self=Get EditorActorSubsystem(), ActorsToSelect=@n5)
   Add(execute=@n4.LoopBody, TargetArray=@n5, NewItem=@n4)
  ]}
 ToggleEditorVisibility {nodes [
   n1 = ToggleEditorVisibility!() -> (then, ActorClasses:Array of Actor Class References, bIsHidden:bool=~) {is=title K2Node_FunctionEntry}
   > n2 = For Each Loop(Array:Array of Actor Class References=@n1) -> (Array Element:Actor Class Reference)
   > n3 = GetAllActorsOfClass(ActorClass=@n2)
   > n4 = For Each Loop(Array:Array of Actor Object References=@n3) -> (Array Element:Actor Object Reference)
   > SetIsTemporarilyHiddenInEditor(self=@n4, bIsHidden=@n1.bIsHidden)
  ]}
]
variables [
 AR_ActorClasses {type="array<AssetData>", default []}
 AR_Actors {type="array<BuildingActor>", default []}
 AR_BSG_Actors {type="array<Actor>"}
 AR_EmptyActors {type="array<Actor>"}
 AR_EnvActors {type="array<Actor>"}
 AR_FSM_Actors {type="array<Actor>"}
 AR_SoftPaths {type="array<SoftObjectPath>", default []}
 AR_Valid_FSM {type="array<Class>", default []}
 AR_ValidActors {type="array<Actor>"}
 AR_ValidArchitectureClasses {
  type="array<Class>"
  default [
   &BuildingWall
   &BuildingFloor
   /Script/FortniteGame.BuildingPillar
   /Script/FortniteGame.BuildingPropWall
   &BuildingRoof
   &BuildingStairs
   &BuildingStairs
   null
  ]
 }
 AR_ValidDamageClasses {
  type="array<Class>"
  default [
   &BuildingWall
   &BuildingFloor
   &BuildingPillar
   &BuildingPropWall
   &BuildingRoof
   &BuildingStairs
   &BuildingStairs
   &BuildingProp
  ]
 }
 AR_ValidPropClasses {type="array<Class>", default [&BuildingProp, /Script/FortniteGame.BuildingDeco]}
 BuildingStaticGeo {type=Class, default=&BuildingStaticGeo}
 CollectionName {type=string, default=ENV_SelectedArchitecture}
 CurrentBuildingActor {type=BuildingActor, default=null}
 CurrentTODM {type=Class, default=/Game/TimeOfDay/DaySequence/DaySequenceActors/BR_Ch6/DSA_BR_Ch6S1.DSA_BR_Ch6S1_C}
 EmissiveIntensity {type=number, default=1}
 IsWireframeEnabled {type=boolean}
 ModifiedActor {type=Actor, default=null}
 ModifiedMeshComponent {type=StaticMeshComponent, default=null}
 NewOverlayMaterial {type=MaterialInstanceDynamic}
 OriginalTransform {type=Transform, default {location=<0 0 0>, rotation=<rot 0 0 0>, scale=<1 1 1>}}
 OverlayMaterial {type=MaterialInterface, default=/Game/ArtTools/Environment/OverlayMaterials/MI_WireFrame_Outline}
 ReplaceBaseMaterial {type=boolean}
 SelectedTextureData {type=BuildingTextureData, default=null}
 StaticMeshClass {type=Class, default=/Script/FortniteGame.FortStaticMeshActor}
 TextureData {type=BuildingTextureData, default=null}
 WireframeColor {type=Color, default=<rgba 0 0.8706 1 1>}
]
widgets {
 info {rootWidgetClass=&VerticalBox, widgetCount=28}
 widgets [
  VerticalBox_2 {widgetClassPath=&VerticalBox, widget=VerticalBox_2}
  HorizontalBox_0 {widgetClassPath=&HorizontalBox, parent=VerticalBox_2, widget=HorizontalBox_0, slot {class=VerticalBoxSlot}}
  VerticalBox_0 {widgetClassPath=&VerticalBox, parent=HorizontalBox_0, widget=VerticalBox_0, slot {class=HorizontalBoxSlot}}
  BTN_HideAllArchitecture {
   widgetClassPath=&TA_Generic01_Button
   parent=VerticalBox_0
   widget=BTN_HideAllArchitecture
   properties {autoToggleSelection=true, label=Hide Architecture, sizeOverride=<130 0>}
   slot {class=VerticalBoxSlot, properties {padding=<ltrb 5 5 5 5>}}
  }
  BTN_HideAllProps {
   widgetClassPath=&TA_Generic01_Button
   parent=VerticalBox_0
   widget=BTN_HideAllProps
   properties {autoToggleSelection=true, label=Hide Props, sizeOverride=<130 0>}
   slot {class=VerticalBoxSlot, properties {padding=<ltrb 5 5 5 5>}}
  }
  BTN_HideStaticGeo {
   widgetClassPath=&TA_Generic01_Button
   parent=VerticalBox_0
   widget=BTN_HideStaticGeo
   properties {autoToggleSelection=true, label=Hide Static Geo, sizeOverride=<130 0>}
   slot {class=VerticalBoxSlot, properties {padding=<ltrb 5 5 5 5>}}
  }
  VerticalBox_1 {widgetClassPath=&VerticalBox, parent=HorizontalBox_0, widget=VerticalBox_1, slot {class=HorizontalBoxSlot}}
  BTN_SelectArchitecture {
   widgetClassPath=&TA_Generic01_Button
   parent=VerticalBox_1
   widget=BTN_SelectArchitecture
   properties {label=Select Architecture, sizeOverride=<130 0>}
   slot {class=VerticalBoxSlot, properties {padding=<ltrb 5 5 5 5>}}
  }
  BTN_SelectProps {
   widgetClassPath=&TA_Generic01_Button
   parent=VerticalBox_1
   widget=BTN_SelectProps
   properties {label=Select Props, sizeOverride=<130 0>}
   slot {class=VerticalBoxSlot, properties {padding=<ltrb 5 5 5 5>}}
  }
  BTN_SelectByProperty {
   widgetClassPath=&TA_Generic01_Button
   parent=VerticalBox_1
   widget=BTN_SelectByProperty
   properties {label="Select CanBeDamaged = False", sizeOverride=<130 0>}
   slot {class=VerticalBoxSlot, properties {padding=<ltrb 5 5 5 5>}}
  }
  VerticalBox {widgetClassPath=&VerticalBox, parent=HorizontalBox_0, widget=VerticalBox, slot {class=HorizontalBoxSlot}}
  BTN_SelectStaticMeshes {
   widgetClassPath=&TA_Generic01_Button
   parent=VerticalBox
   widget=BTN_SelectStaticMeshes
   properties {label=Select Static Meshes, sizeOverride=<130 0>}
   slot {class=VerticalBoxSlot, properties {padding=<ltrb 5 5 5 5>}}
  }
  BTN_SelectStaticGeo {
   widgetClassPath=&TA_Generic01_Button
   parent=VerticalBox
   widget=BTN_SelectStaticGeo
   properties {label=Select Static Geo, sizeOverride=<130 0>}
   slot {class=VerticalBoxSlot, properties {padding=<ltrb 5 5 5 5>}}
  }
  BTN_UpdateTODM {
   widgetClassPath=&TA_Generic01_Button
   parent=VerticalBox
   widget=BTN_UpdateTODM
   properties {label=Update TODM, sizeOverride=<130 0>}
   slot {class=VerticalBoxSlot, properties {padding=<ltrb 5 5 5 5>}}
  }
  BTN_EditWidget {
   widgetClassPath=&TA_Generic01_Button
   parent=HorizontalBox_0
   widget=BTN_EditWidget
   properties {label=">>", buttonColor_Default=<rgba 0.0469 0.0469 0.0469 1>, sizeOverride=<20 0>}
   slot {class=HorizontalBoxSlot, properties {padding=<ltrb 5 5 5 5>}}
  }
  HB_WireframeOverlay {widgetClassPath=&HorizontalBox, parent=VerticalBox_2, widget=HB_WireframeOverlay, slot {class=VerticalBoxSlot}}
  BTN_EnableOverlay {
   widgetClassPath=&TA_Generic01_Button
   parent=HB_WireframeOverlay
   widget=BTN_EnableOverlay
   properties {autoToggleSelection=true, label=View Wireframe, sizeOverride=<130 0>}
   slot {class=HorizontalBoxSlot, properties {horizontalAlignment=HAlign_Center, verticalAlignment=VAlign_Center, padding=<ltrb 5 5 5 5>}}
  }
  VerticalBox_3 {widgetClassPath=&VerticalBox, parent=HB_WireframeOverlay, widget=VerticalBox_3, slot {class=HorizontalBoxSlot}}
  SPV_WireframeColor {
   widgetClassPath=&SinglePropertyView
   parent=VerticalBox_3
   widget=SPV_WireframeColor
   properties {propertyName=WireframeColor}
   slot {class=VerticalBoxSlot}
  }
  SPV_ReplaceBaseMat {
   widgetClassPath=&SinglePropertyView
   parent=VerticalBox_3
   widget=SPV_ReplaceBaseMat
   properties {propertyName=ReplaceBaseMaterial}
   slot {class=VerticalBoxSlot}
  }
  VerticalBox_4 {widgetClassPath=&VerticalBox, parent=VerticalBox_2, widget=VerticalBox_4, slot {class=VerticalBoxSlot}}
  HorizontalBox_1 {widgetClassPath=&HorizontalBox, bIsVariable=true, parent=VerticalBox_4, widget=HorizontalBox_1, slot {class=VerticalBoxSlot}}
  HorizontalBox_2 {widgetClassPath=&HorizontalBox, parent=VerticalBox_4, widget=HorizontalBox_2, slot {class=VerticalBoxSlot}}
  BTN_SelectByTextureData {
   widgetClassPath=&TA_Generic01_Button
   parent=HorizontalBox_2
   widget=BTN_SelectByTextureData
   properties {label=Select by Texture Data, sizeOverride=<130 0>}
   slot {class=HorizontalBoxSlot, properties {horizontalAlignment=HAlign_Center, verticalAlignment=VAlign_Center, padding=<ltrb 5 5 5 5>}}
  }
  SPV_TextureData {
   widgetClassPath=&SinglePropertyView
   parent=HorizontalBox_2
   widget=SPV_TextureData
   properties {propertyName=TextureData}
   slot {class=HorizontalBoxSlot}
  }
  VB_Version2 {widgetClassPath=&VerticalBox, bIsVariable=true, parent=VerticalBox_2, widget=VB_Version2, slot {class=VerticalBoxSlot}}
  UEFN_TextBlock_C_121 {
   widgetClassPath=&UEFN_TextBlock
   parent=VB_Version2
   widget=UEFN_TextBlock_C_121
   properties {text=Version 2}
   slot {class=VerticalBoxSlot}
  }
  DV_ActorClasses {widgetClassPath=&DetailsView, parent=VB_Version2, widget=DV_ActorClasses, slot {class=VerticalBoxSlot}}
 ]
}
